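# Conversation database for the Agent Peely NPC, defining the dialogue a player gets when talking to him.
# The entry point leads to one speech node with a default line, overridden by prioritized contextual lines for first-ever meetings and specific player outfits.
# That speech branches to link nodes into the shared NPC quest, basic services and a Season 26 Snapshot quest conversations.
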ConversationDatabase /NPCLibrary/NPCs/AgentPeely/NPCLibrary_Conversation_AgentPeely
entry_points [{tag=Conversation.Entry.AgentPeely, destinations [#1]}]
nodes [
 ConversationEntryPointNode_7 {
  id=#1
  class=ConversationEntryPointNode
  outputs [#2]
  properties {entryTag {tagName=Conversation.Entry.AgentPeely}}
 }
 Speech_C_4 {
  id=#2
  class=Speech_C
  outputs [#3, #4, #5]
  properties {
   nodeName=""
   generalConfig {
    defaultMessage=(Agent Peely pats you down for weapons and wires.)
    contextualMessages [
     {
      message=(Agent Peely is making a gadget with a banana and a calculator wristwatch.)
      contextRequirements [{requirement {class=FortControllerRequirement_IsFirstEverConversationWithNPC}}]
      priority {curve {rowName=Tandem.Speech.Priority.FirstEverInteraction}}
     }
     {
      message=(Agent Peely gives you a knowing secret agent nod.)
      contextRequirements [{requirement {
         class=FortControllerRequirement_HasCID
         properties {allowedCIds [&CID_693_Athena_Commando_M_BuffCat]}
        }}]
      priority {curve {rowName=Tandem.Speech.Priority.OutfitReaction}}
     }
     {
      message=(Agent Peely doesn't register you as suspicious at all.)
      contextRequirements [{requirement {class=FortControllerRequirement_HasCID, properties {allowedCIds [&Character_MusketSlinger]}}}]
      priority {curve {rowName=Tandem.Speech.Priority.OutfitReaction}}
     }
     {
      message=(Agent Peely vows to avenge your freezing.)
      contextRequirements [{requirement {
         class=FortControllerRequirement_HasCID
         properties {allowedCIds [&CID_A_323_Athena_Commando_M_BananaWinter]}
        }}]
      priority {curve {rowName=Tandem.Speech.Priority.OutfitReaction}}
     }
     {
      message=(Agent Peely admires your personal style.)
      contextRequirements [{requirement {class=FortControllerRequirement_HasCID, properties {allowedCIds [/BRCosmetics/Athena/Items/Cosmetics/Characters/Character_LazarusLens]}}}]
      priority {curve {rowName=Tandem.Speech.Priority.OutfitReaction}}
     }
    ]
   }
  }
 }
 ConversationLinkNode_2 {
  id=#5
  class=ConversationLinkNode
  properties {remoteEntryTag {tagName=Conversation.Entry.Season26.Snapshot.P01.Q012.a}}
 }
 ConversationLinkNode_0 {id=#4, class=ConversationLinkNode, properties {remoteEntryTag {tagName=Conversation.Entry.BasicServices}}}
 ConversationLinkNode_1 {id=#3, class=ConversationLinkNode, properties {remoteEntryTag {tagName=Conversation.Entry.NPCLibrary.Quests}}}
]
